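# First-person firing sound cue for the Bone SMG, built from layered, randomized close-range samples.
# A mixer blends an immediate shot with slightly delayed growl, later chamber-reset and last tail layers, each at its own volume.
# Each layer picks a random wave, and the delayed layers pass through modulators for variation.
# It references shared first-person weapon attenuation, sound class and gunfire concurrency assets.
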SoundCue /PrimalGameplay/Items/Guns/Bone/BoneSMG/SMG_Bone_Fire_1P_Cue
root=Mixer_0
nodes [
 {id=Delay_0, children [Modulator_0], properties {delayMax=0.1, delayMin=0.1}}
 {id=Delay_2, children [Modulator_4], properties {delayMax=0.01, delayMin=0.01}}
 {id=Delay_4, children [Modulator_3], properties {delayMax=0.22, delayMin=0.22}}
 {id=Mixer_0, children [Random_7,Delay_2,Delay_0,Delay_4], properties {inputVolume [0.6,1,0.45,0.6]}}
 {id=Modulator_0, children [Random_0]}
 {id=Modulator_3, children [Random_5], properties {pitchMax=1.5, volumeMax=1, volumeMin=0.8}}
 {id=Modulator_4, children [Random_2], properties {pitchMax=1, pitchMin=1, volumeMax=1, volumeMin=0.8}}
 {id=Random_0, children [WavePlayer_0,WavePlayer_1], properties {weights [1,1]}}
 {id=Random_2, children [WavePlayer_8, WavePlayer_10, WavePlayer_11], properties {weights [1,1,1]}}
 {id=Random_5, children [WavePlayer_27, WavePlayer_28, WavePlayer_29, WavePlayer_30], properties {weights [1,1,1,1]}}
 {id=Random_7, children [WavePlayer_38, WavePlayer_39, WavePlayer_40, WavePlayer_41], properties {weights [1,1,1,1]}}
 {id=WavePlayer_0, properties {soundWaveAssetPtr=&/SMG_Bone_Fire_Close_Chamber_Reset_01}}
 {id=WavePlayer_1, properties {soundWaveAssetPtr=&/SMG_Bone_Fire_Close_Chamber_Reset_02}}
 {id=WavePlayer_10, properties {soundWaveAssetPtr=&/SMG_Bone_Fire_Close_Growl_03}}
 {id=WavePlayer_11, properties {soundWaveAssetPtr=&/SMG_Bone_Fire_Close_Growl_04}}
 {id=WavePlayer_27, properties {soundWaveAssetPtr=&/SMG_Bone_Fire_Close_Tails_01}}
 {id=WavePlayer_28, properties {soundWaveAssetPtr=&/SMG_Bone_Fire_Close_Tails_02}}
 {id=WavePlayer_29, properties {soundWaveAssetPtr=&/SMG_Bone_Fire_Close_Tails_03}}
 {id=WavePlayer_30, properties {soundWaveAssetPtr=&/SMG_Bone_Fire_Close_Tails_04}}
 {id=WavePlayer_38, properties {soundWaveAssetPtr=&/SMG_Bone_Fire_Close_01}}
 {id=WavePlayer_39, properties {soundWaveAssetPtr=&/SMG_Bone_Fire_Close_02}}
 {id=WavePlayer_40, properties {soundWaveAssetPtr=&/SMG_Bone_Fire_Close_03}}
 {id=WavePlayer_41, properties {soundWaveAssetPtr=&/SMG_Bone_Fire_Close_04}}
 {id=WavePlayer_8, properties {soundWaveAssetPtr=&/SMG_Bone_Fire_Close_Growl_01}}
]
properties {
 attenuationSettings=&Fort_Weapon_Projectile_FirstPerson_Athena_Att
 maxDistance=2097152
 soundClassObject=&Fort_WeaponRanged_Fire_1P_PSM
 volumeMultiplier=1
 attenuationOverrides {distanceAlgorithm=NaturalSound, falloffDistance=5000, attenuationShapeExtents {x=1000}}
 concurrencyOverrides {bLimitToOwner=true, maxCount=3, resolutionRule=StopOldest}
 concurrencySet [&Weapon_Gunfire_FirstPerson_Conc]
}
