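# Sound cue for the bone SMG's gunfire, blending close, medium and distant layers at different volumes, each with its own sound class and attenuation.
# The close layer mixes a random shot with a slightly delayed growl and a later, pitch-varied tail.
# Medium and distant layers pick random variants, the medium one pitched up, and both share the distant fire sound class.
# Local and global gunfire concurrency sets limit overlapping shots.
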SoundCue /PrimalGameplay/Items/Guns/Bone/BoneSMG/SMG_Bone_Fire_Cue
root=Mixer_17
nodes [
 {
  id=Attenuation_0
  children [Modulator_0]
  properties {attenuationSettings=&Fort_Weapon_Projectile_Medium_Mono_Athena_Att}
 }
 {
  id=Attenuation_34
  children [Modulator_1]
  properties {attenuationSettings=&Fort_Weapon_Projectile_Distant_Mono_Athena_Att}
 }
 {id=Attenuation_49, children [Mixer_1], properties {attenuationSettings=&Fort_Weapon_Projectile_Close_Athena_Att}}
 {id=Delay_1, children [Modulator_3], properties {delayMax=0.3, delayMin=0.3}}
 {id=Delay_2, children [Modulator_4], properties {delayMax=0.01, delayMin=0.01}}
 {id=Mixer_1, children [Random_1,Delay_2,Delay_1], properties {inputVolume [0.6,1,0.6]}}
 {id=Mixer_17, children [SoundClass_32, SoundClass_0, SoundClass_35], properties {inputVolume [1,0.4,0.3]}}
 {id=Modulator_0, children [Random_7], properties {pitchMax=1.25, pitchMin=1.2, volumeMax=1, volumeMin=1}}
 {id=Modulator_1, children [Random_8], properties {volumeMax=1, volumeMin=1}}
 {id=Modulator_3, children [Random_5], properties {pitchMax=1.5, volumeMax=1, volumeMin=0.8}}
 {id=Modulator_4, children [Random_2], properties {pitchMax=1, pitchMin=1, volumeMax=1, volumeMin=0.8}}
 {id=Random_1, children [WavePlayer_12, WavePlayer_13, WavePlayer_17, WavePlayer_18], properties {weights [1,1,1,1]}}
 {id=Random_2, children [WavePlayer_8, WavePlayer_10, WavePlayer_11], properties {weights [1,1,1]}}
 {id=Random_5, children [WavePlayer_4,WavePlayer_5,WavePlayer_6,WavePlayer_7], properties {weights [1,1,1,1]}}
 {id=Random_7, children [WavePlayer_14, WavePlayer_15, WavePlayer_16, WavePlayer_22], properties {weights [1,1,1,1]}}
 {id=Random_8, children [WavePlayer_23, WavePlayer_24, WavePlayer_25, WavePlayer_26], properties {weights [1,1,1,1]}}
 {
  id=SoundClass_0
  children [Attenuation_0]
  properties {soundClassOverride=&Fort_WeaponRanged_Fire_Distant}
 }
 {
  id=SoundClass_32
  children [Attenuation_49]
  properties {soundClassOverride=&Fort_WeaponRanged_Fire_3P_PSM}
 }
 {
  id=SoundClass_35
  children [Attenuation_34]
  properties {soundClassOverride=&Fort_WeaponRanged_Fire_Distant}
 }
 {id=WavePlayer_10, properties {soundWaveAssetPtr=&/SMG_Bone_Fire_Close_Growl_03}}
 {id=WavePlayer_11, properties {soundWaveAssetPtr=&/SMG_Bone_Fire_Close_Growl_04}}
 {id=WavePlayer_12, properties {soundWaveAssetPtr=&/SMG_Bone_Fire_Close_01}}
 {id=WavePlayer_13, properties {soundWaveAssetPtr=&/SMG_Bone_Fire_Close_02}}
 {id=WavePlayer_14, properties {soundWaveAssetPtr=&/SMG_Bone_Fire_Mid_01}}
 {id=WavePlayer_15, properties {soundWaveAssetPtr=&/SMG_Bone_Fire_Mid_02}}
 {id=WavePlayer_16, properties {soundWaveAssetPtr=&/SMG_Bone_Fire_Mid_03}}
 {id=WavePlayer_17, properties {soundWaveAssetPtr=&/SMG_Bone_Fire_Close_03}}
 {id=WavePlayer_18, properties {soundWaveAssetPtr=&/SMG_Bone_Fire_Close_04}}
 {id=WavePlayer_22, properties {soundWaveAssetPtr=&/SMG_Bone_Fire_Mid_04}}
 {id=WavePlayer_23, properties {soundWaveAssetPtr=&/SMG_Bone_Fire_Distant_01}}
 {id=WavePlayer_24, properties {soundWaveAssetPtr=&/SMG_Bone_Fire_Distant_02}}
 {id=WavePlayer_25, properties {soundWaveAssetPtr=&/SMG_Bone_Fire_Distant_03}}
 {id=WavePlayer_26, properties {soundWaveAssetPtr=&/SMG_Bone_Fire_Distant_04}}
 {id=WavePlayer_4, properties {soundWaveAssetPtr=&/SMG_Bone_Fire_Close_Tails_01}}
 {id=WavePlayer_5, properties {soundWaveAssetPtr=&/SMG_Bone_Fire_Close_Tails_02}}
 {id=WavePlayer_6, properties {soundWaveAssetPtr=&/SMG_Bone_Fire_Close_Tails_03}}
 {id=WavePlayer_7, properties {soundWaveAssetPtr=&/SMG_Bone_Fire_Close_Tails_04}}
 {id=WavePlayer_8, properties {soundWaveAssetPtr=&/SMG_Bone_Fire_Close_Growl_01}}
]
properties {
 maxDistance=27500
 volumeMultiplier=1
 attenuationOverrides {distanceAlgorithm=NaturalSound, falloffDistance=5000, attenuationShapeExtents {x=1000}}
 concurrencySet [&Weapon_Gunfire_Local_Conc, &Weapon_Gunfire_Global_Conc]
}
