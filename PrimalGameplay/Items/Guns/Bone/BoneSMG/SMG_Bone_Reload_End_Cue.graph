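# Sound cue for the end of the Bone SMG's reload, played as weapon foley.
# A single wave player feeds a modulator that slightly randomizes pitch while keeping volume constant.
# It uses Fortnite's weapon reload attenuation and foley sound class at reduced volume.
# Its own concurrency override prevents new instances from starting once the limit is reached, rather than stopping playing ones.
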SoundCue /PrimalGameplay/Items/Guns/Bone/BoneSMG/SMG_Bone_Reload_End_Cue
root=Modulator_0
nodes [
 {id=Modulator_0, children [WavePlayer_0], properties {pitchMax=1.1, pitchMin=0.9, volumeMax=1, volumeMin=1}}
 {id=WavePlayer_0, properties {soundWaveAssetPtr=&/SMG_Bone_Reload_End}}
]
properties {
 attenuationSettings=&Fort_Weapon_Reload_Att
 bOverrideConcurrency=true
 maxDistance=2250
 soundClassObject=&Fort_WeaponFoley
 volumeMultiplier=0.401
 attenuationOverrides {falloffDistance=4100, attenuationShapeExtents {x=900}}
 concurrencyOverrides {resolutionRule=PreventNew}
}
